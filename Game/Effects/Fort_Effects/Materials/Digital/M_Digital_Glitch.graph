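# M_Digital_Glitch is an unlit translucent particle material for digital glitch effects on sprites.
# Colour is a per-particle pick of two tints or an optional scrolling rainbow, blendable toward plain particle colour.
# Panning digital noise, randomised per particle, distorts two crossed side masks for view-angle-faded opacity and drives a transformed, GlitchAmount-scaled vector.
# Mobile swaps ParticleRandom for a dynamic parameter.
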Material /Game/Effects/Fort_Effects/Materials/Digital/M_Digital_Glitch
graph {
 properties {
  bUsedWithNiagaraSprites=true
  bUsedWithParticleSprites=true
  blendMode=BLEND_Translucent
  pixelDepthOffsetMode=PDOM_Legacy
  shadingModel=MSM_Unlit
  usedShadingModels=Unlit
 }
 nodes [
  n1 = ComponentMask(None=ParticleMacroUV()) {g=true}
  n2 = ParticleRandom()
  n3 = ShadingPathSwitch(Default=@n2, Deferred=@n2, Mobile=DynamicParameter() {paramNames [Random Value, Param2, Param3, Param4]}.Random Value) {desc=Migrated from FeatureLevelSwitch}
  n4 = StaticSwitchParameter(
   True=TextureSample(UVs=Panner(Coordinate=@n1) {speedX=1.6667}) {texture=/Game/Global/Textures/Default/Masks_and_Gradients/T_Grad_Rainbow}.RGB
   False=LinearInterpolate(
    A=VectorParameter() {parameterName=ColorA, defaultValue=<rgba 0 4.7442 10 1>}.RGB
    B=VectorParameter() {parameterName=ColorB, defaultValue=<rgba 0.8833 0 50 1>}.RGB
    Alpha=Round(None=@n3)
   )
  ) {parameterName=Use RainbowFlashing}
  n5 = ParticleColor()
  MaterialFunctionCall(Emissive=LinearInterpolate(A=Multiply(A=@n4, B=@n5.RGB), B=@n5.RGB, Alpha=ScalarParameter() {parameterName=Use Particle Color Only})) {materialFunction=&MF_FX_EmissiveAdjustments}
  n6 = TextureCoordinate() {uTiling=0.01, vTiling=0.01}
  n7 = ParticleRandom()
  n8 = ShadingPathSwitch(Default=@n7, Deferred=@n7, Mobile=DynamicParameter() {paramNames [Random Value, Param2, Param3, Param4]}.Random Value) {desc=Migrated from FeatureLevelSwitch}
  n9 = Time()
  n10 = ScalarParameter() {defaultValue=2, parameterName=Speeed}
  n11 = TextureSample(UVs=Panner(Coordinate=Add(A=@n6, B=@n8), Time=Multiply(A=@n9, B=@n10)) {speedX=0.1, speedY=0.1}) {samplerType=SAMPLERTYPE_LinearColor, texture=/Game/Global/Textures/Default/Noise/T_Atlas_DigitalNoise_01}
  n12 = Multiply(A=ComponentMask(None=@n11.RGB) {g=true, r=true}) {constB=0.025}
  n13 = TextureCoordinate()
  n14 = TextureCoordinate()
  n15 = Add(A=@n12, B=@n14)
  n16 = Multiply(
   A=TextureSample(UVs=Add(A=@n12, B=@n13)) {texture=/Game/Effects/Fort_Effects/Textures/Masks/T_Side_Mask}.G
   B=TextureSample(UVs=MaterialFunctionCall(XYZ=@n15, XY=@n14) {materialFunction=&Swizzle}.YX) {texture=&T_Side_Mask}.G
  )
  n17 = ScalarParameter() {parameterName=Angle Fade}
  Multiply(
   A=Multiply(
    A=Multiply(A=@n5.A, B=@n16)
    B=MaterialFunctionCall("FadeSensitivity (optional)"=@n17) {materialFunction=&FadeBasedOnViewAngle}.Result
   )
   B=ScalarParameter() {defaultValue=1, parameterName=Opacity}
  )
  Multiply(A=Transform(None=@n11.RGB), B=ScalarParameter() {defaultValue=10, parameterName=GlitchAmount})
 ]
}
