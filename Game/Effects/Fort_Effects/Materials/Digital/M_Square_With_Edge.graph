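# Unlit, two-sided translucent material for Niagara sprites that draws a procedural square with a distinct border and center opacity.
# The minimum of UV and inverted UV gives edge distance, which radius and softness parameters turn into a border mask without textures.
# That mask blends border and center opacity, multiplied by particle alpha.
# Particle color RGB feeds the shared MF_FX_EmissiveAdjustments function for emissive.
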Material /Game/Effects/Fort_Effects/Materials/Digital/M_Square_With_Edge
graph {
 properties {
  bUsedWithNiagaraSprites=true
  blendMode=BLEND_Translucent
  pixelDepthOffsetMode=PDOM_Legacy
  shadingModel=MSM_Unlit
  twoSided=true
  usedShadingModels=Unlit
 }
 nodes [
  n1 = ParticleColor()
  MaterialFunctionCall(Emissive=@n1.RGB) {materialFunction=&MF_FX_EmissiveAdjustments}
  n2 = TextureCoordinate()
  n3 = OneMinus(None=@n2)
  n4 = Divide(
   A=Subtract(A=Min(A=@n2, B=@n3), B=ScalarParameter() {defaultValue=0.1, parameterName=BorderRadius}) {constB=0.1}
   B=ScalarParameter() {parameterName=BorderSoftness}
  ) {constB=0.1}
  n5 = Min(A=ComponentMask(None=@n4) {r=true}, B=ComponentMask(None=@n4) {g=true})
  Multiply(
   A=@n1.A
   B=LinearInterpolate(
    A=ScalarParameter() {defaultValue=1, parameterName=BorderOpacity}
    B=ScalarParameter() {defaultValue=0.1, parameterName=CenterOpacity}
    Alpha=Saturate(None=@n5)
   )
  )
 ]
}
